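# Physics asset for the PinkSpike medium female player body, set up for secondary motion on jacket chains and a thigh-mounted gun.
# Capsules on pelvis, spine and thighs form the body volume, with tiny master bodies anchoring the dynamic bones.
# Five three-link chains hang from the jacket master and the gun from the right thigh, with narrow twist and asymmetric swing limits.
# Extra angularly free constraints tie each side chain tip to its front neighbour.
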PhysicsAsset /BRCosmetics/Characters/Player/Female/Medium/Bodies/F_MED_PinkSpike/Meshes/F_MED_PinkSpike_Physics
bodies {
 dyn_L_front_chain_1_jnt [{shapeType=Capsule, length=6, radius=7.5, shapeName=dyn_L_front_chain_1_jnt_capsule, center=<4.9302 0 3.9371>, rotation=<rot 0 0 90>}]
 dyn_L_front_chain_2_jnt [{shapeType=Capsule, length=6, radius=7.5, shapeName=dyn_L_front_chain_2_jnt_capsule, center=<4.7005 0 5.3689>, rotation=<rot 0 0 90>}]
 dyn_L_front_chain_3_jnt [{shapeType=Capsule, length=6, radius=7.5, shapeName=dyn_R_front_chain_3_jnt_capsule, center=<6.1646 0 6.8>, rotation=<rot 0 0 90>}]
 dyn_L_side_chain_1_jnt [{shapeType=Capsule, length=6, radius=7.5, shapeName=dyn_L_side_chain_1_jnt_capsule, center=<7.6199 4.6532 4.3868>, rotation=<rot 0 0 90>}]
 dyn_L_side_chain_2_jnt [{shapeType=Capsule, length=6, radius=7.5, shapeName=dyn_L_side_chain_2_jnt_capsule, center=<7.622 5.4118 4.113>, rotation=<rot 0 0 90>}]
 dyn_L_side_chain_3_jnt [{shapeType=Capsule, length=6, radius=7.5, shapeName=dyn_R_side_chain_3_jnt_capsule, center=<7.623 5.769 3.984>, rotation=<rot 0 0 90>}]
 dyn_R_front_chain_1_jnt [{shapeType=Capsule, length=6, radius=7.5, shapeName=dyn_R_front_chain_1_jnt_capsule, center=<-4.9 0 -4.6603>, rotation=<rot 0 0 90>}]
 dyn_R_front_chain_2_jnt [{shapeType=Capsule, length=6, radius=7.5, shapeName=dyn_R_front_chain_2_jnt_capsule, center=<-4.7 0 -6.1572>, rotation=<rot 0 0 90>}]
 dyn_R_front_chain_3_jnt [{shapeType=Capsule, length=6, radius=7.5, shapeName=dyn_R_front_chain_3_jnt_capsule, center=<-6.1646 0 -5.7621>, rotation=<rot 0 0 90>}]
 dyn_R_side_chain_1_jnt [{shapeType=Capsule, length=6, radius=7.5, shapeName=dyn_L_side_chain_1_jnt_capsule, center=<-7.6199 -4.6532 -4.3868>, rotation=<rot 0 0 90>}]
 dyn_R_side_chain_2_jnt [{shapeType=Capsule, length=6, radius=7.5, shapeName=dyn_L_side_chain_2_jnt_capsule, center=<-7.622 -5.4118 -4.113>, rotation=<rot 0 0 90>}]
 dyn_R_side_chain_3_jnt [{shapeType=Capsule, length=6, radius=7.5, shapeName=dyn_R_side_chain_3_jnt_capsule, center=<-7.623 -5.769 -3.984>, rotation=<rot 0 0 90>}]
 dyn_back_chain_1_jnt [{shapeType=Capsule, length=6, radius=7.5, shapeName=dyn_back_chain_1_jnt_capsule, center=<7.4693 0 6.6475>, rotation=<rot 0 0 90>}]
 dyn_back_chain_2_jnt [{shapeType=Capsule, length=6, radius=7.5, shapeName=dyn_back_chain_2_jnt_capsule, center=<7.4693 0 5.4276>, rotation=<rot 0 0 90>}]
 dyn_back_chain_3_jnt [{shapeType=Capsule, length=6, radius=7.5, shapeName=dyn_back_chain_3_jnt_capsule, center=<7.4693 0 5.4498>, rotation=<rot 0 0 90>}]
 dyn_gun_jnt [{
   shapeType=Capsule
   length=15.0377
   radius=9.8846
   shapeName=dyn_gun_jnt_capsule
   center=<2.3492 -7.2153 -6.5981>
   rotation=<rot 1.008 -85.9403 -87.2291>
  }]
 master_L_front_chain_1_jnt [{shapeType=Capsule, length=0.505, radius=0.505, shapeName=master_L_front_chain_1_jnt_capsule, rotation=<rot 0 -90 0>}]
 master_L_side_chain_1_jnt [{shapeType=Capsule, length=0.505, radius=0.505, shapeName=master_L_side_chain_1_jnt_capsule, rotation=<rot 0 -90 0>}]
 master_R_front_chain_1_jnt [{shapeType=Capsule, length=0.505, radius=0.505, shapeName=master_R_front_chain_1_jnt_capsule, rotation=<rot 0 -90 0>}]
 master_R_side_chain_1_jnt [{shapeType=Capsule, length=0.505, radius=0.505, shapeName=master_R_side_chain_1_jnt_capsule, rotation=<rot 0 -90 0>}]
 master_back_chain_1_jnt [{shapeType=Capsule, length=0.505, radius=0.505, shapeName=master_back_chain_1_jnt_capsule, rotation=<rot 0 -90 0>}]
 master_gun_jnt [{shapeType=Capsule, length=0.505, radius=0.505, shapeName=master_gun_jnt_capsule, rotation=<rot 0 -90 0>}]
 master_jacket_jnt [{shapeType=Capsule, length=0.505, radius=0.505, shapeName=master_jacket_jnt_capsule, rotation=<rot 0 -90 0>}]
 pelvis [{shapeType=Capsule, length=12.5, radius=10.25, shapeName=pelvis_capsule, center=<-1.4243 -1.699 0>}]
 spine_01 [{shapeType=Capsule, length=9.25, radius=9.75, shapeName=spine_01_capsule, center=<-2.4478 -0.7869 0>}]
 spine_02 [{shapeType=Capsule, length=4, radius=8.25, shapeName=spine_02_capsule, center=<-1.6909 0.8878 0>}]
 spine_03 [{shapeType=Capsule, length=3.5, radius=8.25, shapeName=spine_03_capsule, center=<0 2.6499 0>}]
 spine_04 [{shapeType=Capsule, length=3, radius=11, shapeName=spine_04_capsule, center=<0 3.2166 0>}]
 spine_05 [{
   shapeType=Capsule
   length=11.0422
   radius=10.0815
   shapeName=spine_05_capsule
   center=<3.1535 0.3635 -0.0056>
   rotation=<rot 0 -162.0582 -179.9524>
  }]
 thigh_l [{shapeType=Capsule, length=33.5, radius=8, shapeName=thigh_l_capsule, center=<-20.4354 0.0001 0.1535>, rotation=<rot 90 -179.9999 -179.9997>}]
 thigh_r [{shapeType=Capsule, length=33.5, radius=8, shapeName=thigh_l_capsule, center=<20 -0.5207 0.1535>, rotation=<rot 90 -179.9999 -179.9997>}]
}
constraints [
 {
  bone1Name=spine_01
  bone2Name=pelvis
  swing1LimitDegrees=45
  swing1Motion=Free
  swing2LimitDegrees=45
  swing2Motion=Free
  twistLimitDegrees=45
  twistMotion=Free
 }
 {
  bone1Name=spine_02
  bone2Name=spine_01
  swing1LimitDegrees=45
  swing1Motion=Free
  swing2LimitDegrees=45
  swing2Motion=Free
  twistLimitDegrees=45
  twistMotion=Free
 }
 {
  bone1Name=spine_03
  bone2Name=spine_02
  swing1LimitDegrees=45
  swing1Motion=Free
  swing2LimitDegrees=45
  swing2Motion=Free
  twistLimitDegrees=45
  twistMotion=Free
 }
 {
  bone1Name=spine_04
  bone2Name=spine_03
  swing1LimitDegrees=45
  swing1Motion=Free
  swing2LimitDegrees=45
  swing2Motion=Free
  twistLimitDegrees=45
  twistMotion=Free
 }
 {
  bone1Name=master_jacket_jnt
  bone2Name=spine_04
  swing1LimitDegrees=45
  swing1Motion=Free
  swing2LimitDegrees=45
  swing2Motion=Free
  twistLimitDegrees=45
  twistMotion=Free
 }
 {
  bone1Name=master_L_front_chain_1_jnt
  bone2Name=master_jacket_jnt
  swing1LimitDegrees=45
  swing1Motion=Free
  swing2LimitDegrees=45
  swing2Motion=Free
  twistLimitDegrees=45
  twistMotion=Free
 }
 {bone1Name=dyn_L_front_chain_1_jnt, bone2Name=master_L_front_chain_1_jnt, swing1LimitDegrees=15, swing2LimitDegrees=65, twistLimitDegrees=1}
 {bone1Name=dyn_L_front_chain_2_jnt, bone2Name=dyn_L_front_chain_1_jnt, swing1LimitDegrees=15, swing2LimitDegrees=65, twistLimitDegrees=1}
 {bone1Name=dyn_L_front_chain_3_jnt, bone2Name=dyn_L_front_chain_2_jnt, swing1LimitDegrees=15, swing2LimitDegrees=65, twistLimitDegrees=1}
 {
  bone1Name=master_L_side_chain_1_jnt
  bone2Name=master_jacket_jnt
  swing1LimitDegrees=45
  swing1Motion=Free
  swing2LimitDegrees=45
  swing2Motion=Free
  twistLimitDegrees=45
  twistMotion=Free
 }
 {bone1Name=dyn_L_side_chain_1_jnt, bone2Name=master_L_side_chain_1_jnt, swing1LimitDegrees=15, swing2LimitDegrees=65, twistLimitDegrees=1}
 {bone1Name=dyn_L_side_chain_2_jnt, bone2Name=dyn_L_side_chain_1_jnt, swing1LimitDegrees=15, swing2LimitDegrees=65, twistLimitDegrees=1}
 {bone1Name=dyn_L_side_chain_3_jnt, bone2Name=dyn_L_side_chain_2_jnt, swing1LimitDegrees=15, swing2LimitDegrees=65, twistLimitDegrees=1}
 {
  bone1Name=master_R_front_chain_1_jnt
  bone2Name=master_jacket_jnt
  swing1LimitDegrees=45
  swing1Motion=Free
  swing2LimitDegrees=45
  swing2Motion=Free
  twistLimitDegrees=45
  twistMotion=Free
 }
 {bone1Name=dyn_R_front_chain_1_jnt, bone2Name=master_R_front_chain_1_jnt, swing1LimitDegrees=15, swing2LimitDegrees=65, twistLimitDegrees=1}
 {bone1Name=dyn_R_front_chain_2_jnt, bone2Name=dyn_R_front_chain_1_jnt, swing1LimitDegrees=15, swing2LimitDegrees=65, twistLimitDegrees=1}
 {bone1Name=dyn_R_front_chain_3_jnt, bone2Name=dyn_R_front_chain_2_jnt, swing1LimitDegrees=15, swing2LimitDegrees=65, twistLimitDegrees=1}
 {
  bone1Name=master_R_side_chain_1_jnt
  bone2Name=master_jacket_jnt
  swing1LimitDegrees=45
  swing1Motion=Free
  swing2LimitDegrees=45
  swing2Motion=Free
  twistLimitDegrees=45
  twistMotion=Free
 }
 {bone1Name=dyn_R_side_chain_1_jnt, bone2Name=master_R_side_chain_1_jnt, swing1LimitDegrees=15, swing2LimitDegrees=65, twistLimitDegrees=1}
 {bone1Name=dyn_R_side_chain_2_jnt, bone2Name=dyn_R_side_chain_1_jnt, swing1LimitDegrees=15, swing2LimitDegrees=65, twistLimitDegrees=1}
 {bone1Name=dyn_R_side_chain_3_jnt, bone2Name=dyn_R_side_chain_2_jnt, swing1LimitDegrees=15, swing2LimitDegrees=65, twistLimitDegrees=1}
 {
  bone1Name=master_back_chain_1_jnt
  bone2Name=master_jacket_jnt
  swing1LimitDegrees=45
  swing1Motion=Free
  swing2LimitDegrees=45
  swing2Motion=Free
  twistLimitDegrees=45
  twistMotion=Free
 }
 {bone1Name=dyn_back_chain_1_jnt, bone2Name=master_back_chain_1_jnt, swing1LimitDegrees=15, swing2LimitDegrees=65, twistLimitDegrees=1}
 {bone1Name=dyn_back_chain_2_jnt, bone2Name=dyn_back_chain_1_jnt, swing1LimitDegrees=15, swing2LimitDegrees=65, twistLimitDegrees=1}
 {bone1Name=dyn_back_chain_3_jnt, bone2Name=dyn_back_chain_2_jnt, swing1LimitDegrees=15, swing2LimitDegrees=65, twistLimitDegrees=1}
 {
  bone1Name=spine_05
  bone2Name=spine_04
  swing1LimitDegrees=45
  swing1Motion=Free
  swing2LimitDegrees=45
  swing2Motion=Free
  twistLimitDegrees=45
  twistMotion=Free
 }
 {
  bone1Name=master_gun_jnt
  bone2Name=thigh_r
  swing1LimitDegrees=45
  swing1Motion=Free
  swing2LimitDegrees=45
  swing2Motion=Free
  twistLimitDegrees=45
  twistMotion=Free
 }
 {
  bone1Name=thigh_l
  bone2Name=pelvis
  swing1LimitDegrees=45
  swing1Motion=Free
  swing2LimitDegrees=45
  swing2Motion=Free
  twistLimitDegrees=45
  twistMotion=Free
 }
 {
  bone1Name=thigh_r
  bone2Name=pelvis
  swing1LimitDegrees=45
  swing1Motion=Free
  swing2LimitDegrees=45
  swing2Motion=Free
  twistLimitDegrees=45
  twistMotion=Free
 }
 {
  bone1Name=dyn_R_side_chain_3_jnt
  bone2Name=dyn_R_front_chain_3_jnt
  swing1LimitDegrees=45
  swing1Motion=Free
  swing2LimitDegrees=45
  swing2Motion=Free
  twistLimitDegrees=45
  twistMotion=Free
 }
 {
  bone1Name=dyn_L_side_chain_3_jnt
  bone2Name=dyn_L_front_chain_3_jnt
  swing1LimitDegrees=45
  swing1Motion=Free
  swing2LimitDegrees=45
  swing2Motion=Free
  twistLimitDegrees=45
  twistMotion=Free
 }
 {bone1Name=dyn_gun_jnt, bone2Name=master_gun_jnt, swing1LimitDegrees=40, swing2LimitDegrees=7, twistLimitDegrees=1}
]
