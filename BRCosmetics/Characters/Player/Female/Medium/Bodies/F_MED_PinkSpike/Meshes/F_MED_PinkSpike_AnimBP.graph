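# Post-copy animation Blueprint for the PinkSpike female body cosmetic, driving its jacket, gun joint and spike bones.
# It copies the parent mesh pose through a deform Control Rig, applies corrective bone tweaks, then simulates RigidBody physics in pelvis space.
# Post-sim offsets handle skydiving and the custom emote, then weighted constraints position buckles and spikes.
# Event Graph derives alphas from thigh twist, skydive, front-end and emote montage position; the gun AnimDynamics node is unconnected.
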AnimBlueprint /BRCosmetics/Characters/Player/Female/Medium/Bodies/F_MED_PinkSpike/Meshes/F_MED_PinkSpike_AnimBP
target_skeleton=&/F_MED_PinkSpike_Skeleton
graphs [
 AnimGraph {nodes [
   Anim Dynamics - Bone: dyn_gun_jnt(Alpha=1, ExternalForce="0,0,0") {
    angularDampingOverride=0.9
    bOverrideAngularDamping=true
    bUseGravityOverride=true
    simulationSpace=BoneRelative
    alphaBoolBlend {blendOption=Linear}
    boundBone {boneName=dyn_gun_jnt}
    componentAppliedLinearAccClamp=<3000 3000 300>
    componentLinearAccScale=<0.3 0.2 0.2>
    componentLinearVelScale=<0.3 0.2 0.2>
    gravityOverride {z=-980}
    physicsBodyDefinitions [{
      boundBone {boneName=dyn_gun_jnt}
      boxExtents=<25 25 25>
      constraintSetup {angularLimitsMax=<10 2 15>, angularLimitsMin=<-10 -2 -15>, angularTarget=<1 0 0>}
      localJointOffset=<3 0 0>
     }]
    relativeSpaceBone {boneName=pelvis}
   }
   n1 = Look At - Bone: master_L_front_chain_1_jnt(
    ComponentPose=Local To Component(LocalPose=Use cached pose 'PinkSpikePreSim'())
    Alpha=Get Jacket Front Side Correction() -> (Jacket Front Side Correction:double)
   ) {
    alphaBoolBlend {blendOption=Linear}
    boneToModify {boneName=master_L_front_chain_1_jnt}
    lookAtLocation=<-1.0087 1.3706 -16.508>
    lookAtTarget {boneReference {boneName=thigh_l}}
    lookAt_Axis {axis=<1 0>}
    lookUp_Axis {axis=<0 1>}
   }
   n2 = Get Jacket Front Side Correction()
   n3 = "Transform (Modify) Bone - Bone: master_L_side_chain_1_jnt"(ComponentPose=@n1, Alpha=@n2, Translation="0,0,0", Rotation="-20,0,0", Scale="1,1,1") {
    rotationMode=BMM_Additive
    rotationSpace=BCS_BoneSpace
    alphaBoolBlend {blendOption=Linear}
    boneToModify {boneName=master_L_side_chain_1_jnt}
    rotation {pitch=-20}
   }
   n4 = "Transform (Modify) Bone - Bone: master_L_front_chain_1_jnt"(ComponentPose=@n3, Alpha=@n2, Translation="0,0,0", Rotation="0.077,-0.005,-90.000", Scale="0,1,1") {
    rotationMode=BMM_Additive
    rotationSpace=BCS_BoneSpace
    alphaBoolBlend {blendOption=Linear}
    boneToModify {boneName=master_L_front_chain_1_jnt}
    rotation=<rot 0.077 -0.005 -90>
    scale {x=0}
   }
   n5 = "Transform (Modify) Bone - Bone: master_gun_jnt"(
    ComponentPose=@n4
    Alpha=Get Is In Front End() -> (Is In Front End:double)
    Translation="0,0,0"
    Rotation="-0.000,10.000,20.000"
    Scale="1,1,1"
   ) {
    rotationMode=BMM_Additive
    rotationSpace=BCS_BoneSpace
    alphaBoolBlend {blendOption=Linear}
    boneToModify {boneName=master_gun_jnt}
    rotation=<rot 0 10 20>
   }
   n6 = RigidBody(ComponentPose=@n5) {
    bClampLinearTranslationLimitToRefPose=true
    bOverrideWorldGravity=true
    overridePhysicsAsset=&/F_MED_PinkSpike_Physics
    simulationSpace=BaseBoneSpace
    alphaBoolBlend {blendOption=Linear}
    baseBoneRef {boneName=pelvis}
    componentAppliedLinearAccClamp=<3000 3000 3000>
    componentLinearAccScale=<0.35 0.2 0.2>
    componentLinearVelScale=<0.35 0.2 0.2>
    overrideWorldGravity {z=-980}
    simSpaceSettings {maxAngularAcceleration=1000, maxAngularVelocity=3, maxLinearAcceleration=250, maxLinearVelocity=500, worldAlpha=1}
   }
   n7 = Get Skydiving() -> (Skydiving:double)
   n8 = "Transform (Modify) Bone - Bone: master_back_chain_1_jnt"(ComponentPose=@n6, Alpha=@n7, Translation="0,0,0", Rotation="20,-0.000,-0.000", Scale="1,1,1") {
    rotationMode=BMM_Additive
    rotationSpace=BCS_BoneSpace
    alphaBoolBlend {blendOption=Linear}
    boneToModify {boneName=master_back_chain_1_jnt}
    rotation {pitch=20}
   }
   n9 = "Transform (Modify) Bone - Bone: master_L_side_chain_1_jnt"(ComponentPose=@n8, Alpha=@n7, Translation="0,0,0", Rotation="20,0,0", Scale="1,1,1") {
    rotationMode=BMM_Additive
    rotationSpace=BCS_BoneSpace
    alphaBoolBlend {blendOption=Linear}
    boneToModify {boneName=master_L_side_chain_1_jnt}
    rotation {pitch=20}
   }
   n10 = "Transform (Modify) Bone - Bone: master_R_side_chain_1_jnt"(ComponentPose=@n9, Alpha=@n7, Translation="0,0,0", Rotation="20,0.000,-0.000", Scale="1,1,1") {
    rotationMode=BMM_Additive
    rotationSpace=BCS_BoneSpace
    alphaBoolBlend {blendOption=Linear}
    boneToModify {boneName=master_R_side_chain_1_jnt}
    rotation {pitch=20}
   }
   n11 = Get Is Custom Emote Playing() -> (Is Custom Emote Playing:double)
   n12 = "Transform (Modify) Bone - Bone: master_jacket_jnt"(ComponentPose=@n10, Alpha=@n11, Translation="0,0,0", Rotation="0,-5,0", Scale="1,1,1") {
    rotationMode=BMM_Additive
    rotationSpace=BCS_BoneSpace
    alphaBoolBlend {blendOption=Linear}
    alphaScaleBiasClamp {bInterpResult=true, interpSpeedDecreasing=3, interpSpeedIncreasing=1}
    boneToModify {boneName=master_jacket_jnt}
    rotation {yaw=-5}
   }
   n13 = "Transform (Modify) Bone - Bone: master_L_side_chain_1_jnt"(ComponentPose=@n12, Alpha=@n11, Translation="0,0,0", Rotation="0,10,0", Scale="1,1,1") {
    rotationMode=BMM_Additive
    rotationSpace=BCS_BoneSpace
    alphaBoolBlend {blendOption=Linear}
    alphaScaleBiasClamp {bInterpResult=true, interpSpeedDecreasing=3, interpSpeedIncreasing=1}
    boneToModify {boneName=master_L_side_chain_1_jnt}
    rotation {yaw=10}
   }
   n14 = "Transform (Modify) Bone - Bone: master_L_front_chain_1_jnt"(ComponentPose=@n13, Alpha=@n11, Translation="0,0,0", Rotation="0,30,0", Scale="0,1,1") {
    rotationMode=BMM_Additive
    rotationSpace=BCS_BoneSpace
    alphaBoolBlend {blendOption=Linear}
    alphaScaleBiasClamp {bInterpResult=true, interpSpeedDecreasing=3, interpSpeedIncreasing=1}
    boneToModify {boneName=master_L_front_chain_1_jnt}
    rotation {yaw=30}
    scale {x=0}
   }
   n15 = "Transform (Modify) Bone - Bone: master_R_side_chain_1_jnt"(ComponentPose=@n14, Alpha=@n11, Translation="0,0,0", Rotation="-0.010,-0.008,-0.298", Scale="1,1,1") {
    rotationMode=BMM_Additive
    rotationSpace=BCS_BoneSpace
    alphaBoolBlend {blendOption=Linear}
    alphaScaleBiasClamp {bInterpResult=true, interpSpeedDecreasing=3, interpSpeedIncreasing=1}
    boneToModify {boneName=master_R_side_chain_1_jnt}
    rotation=<rot -0.01 -0.008 -0.298>
   }
   n16 = "Transform (Modify) Bone - Bone: master_R_front_chain_1_jnt"(ComponentPose=@n15, Alpha=@n11, Translation="0,0,0", Rotation="15,20,-10", Scale="1,1,1") {
    rotationMode=BMM_Additive
    rotationSpace=BCS_BoneSpace
    alphaBoolBlend {blendOption=Linear}
    alphaScaleBiasClamp {bInterpResult=true, interpSpeedDecreasing=3, interpSpeedIncreasing=1}
    boneToModify {boneName=master_R_front_chain_1_jnt}
    rotation=<rot 15 20 -10>
   }
   n17 = "Constraint - L_jacket_buckle_button_jnt (PPPPPPPP)"!(
    ComponentPose=@n16
    Alpha=1
    ConstraintWeights_0=0.0003
    ConstraintWeights_1:float=0.0016
    ConstraintWeights_2:float=0.1781
    ConstraintWeights_3:float=0.5141
    ConstraintWeights_4:float=0.0318
    ConstraintWeights_5:float=0.0483
    ConstraintWeights_6:float=0.039
    ConstraintWeights_7:float=0.1868
   ) -> (Pose) {
    alphaBoolBlend {blendOption=Linear}
    boneToModify {boneName=L_jacket_buckle_button_jnt}
    constraintSetup [
     {targetBone {boneName=dyn_R_side_chain_3_jnt}}
     {targetBone {boneName=dyn_L_side_chain_1_jnt}}
     {targetBone {boneName=dyn_L_side_chain_2_jnt}}
     {targetBone {boneName=dyn_L_side_chain_3_jnt}}
     {targetBone {boneName=dyn_back_chain_2_jnt}}
     {targetBone {boneName=dyn_back_chain_3_jnt}}
     {targetBone {boneName=dyn_L_front_chain_2_jnt}}
     {targetBone {boneName=dyn_L_front_chain_3_jnt}}
    ]
    constraintWeights [1,1,1,1,1,1,1,1]
   }
   n18 = "Constraint - R_jacket_buckle_button_jnt (PPPPPPPP)"!(
    ComponentPose=@n17
    Alpha=1
    ConstraintWeights_0=0.0016
    ConstraintWeights_1:float=0.1808
    ConstraintWeights_2:float=0.5226
    ConstraintWeights_3:float=0.0372
    ConstraintWeights_4:float=0.174
    ConstraintWeights_5:float=0.0006
    ConstraintWeights_6:float=0.0317
    ConstraintWeights_7:float=0.0513
   ) -> (Pose) {
    alphaBoolBlend {blendOption=Linear}
    boneToModify {boneName=R_jacket_buckle_button_jnt}
    constraintSetup [
     {targetBone {boneName=dyn_R_side_chain_1_jnt}}
     {targetBone {boneName=dyn_R_side_chain_2_jnt}}
     {targetBone {boneName=dyn_R_side_chain_3_jnt}}
     {targetBone {boneName=dyn_R_front_chain_2_jnt}}
     {targetBone {boneName=dyn_R_front_chain_3_jnt}}
     {targetBone {boneName=dyn_L_side_chain_3_jnt}}
     {targetBone {boneName=dyn_back_chain_2_jnt}}
     {targetBone {boneName=dyn_back_chain_3_jnt}}
    ]
    constraintWeights [1,1,1,1,1,1,1,1]
   }
   n19 = "Constraint - L_lapel_spike_jnt (PPPPPPP)"!(
    ComponentPose=@n18
    Alpha=1
    ConstraintWeights_0=0.1294
    ConstraintWeights_1:float=0.3923
    ConstraintWeights_2:float=0.0689
    ConstraintWeights_3:float=0.1712
    ConstraintWeights_4:float=0.129
    ConstraintWeights_5:float=0.0677
    ConstraintWeights_6:float=0.0416
   ) -> (Pose) {
    alphaBoolBlend {blendOption=Linear}
    boneToModify {boneName=L_lapel_spike_jnt}
    constraintSetup [
     {targetBone {boneName=spine_05}}
     {targetBone {boneName=clavicle_l}}
     {targetBone {boneName=spine_04}}
     {targetBone {boneName=upperarm_twist_01_l}}
     {targetBone {boneName=deform_upperarm_fr_l}}
     {targetBone {boneName=deform_upperarm_up_l}}
     {targetBone {boneName=deform_pec_l}}
    ]
    constraintWeights [1,1,1,1,1,1,1]
   }
   n20 = "Constraint - R_lapel_spike_jnt (PPPPPPP)"!(
    ComponentPose=@n19
    Alpha=1
    ConstraintWeights_0=0.1002
    ConstraintWeights_1:float=0.3623
    ConstraintWeights_2:float=0.0651
    ConstraintWeights_3:float=0.2001
    ConstraintWeights_4:float=0.1809
    ConstraintWeights_5:float=0.0534
    ConstraintWeights_6:float=0.0379
   ) -> (Pose) {
    alphaBoolBlend {blendOption=Linear}
    boneToModify {boneName=R_lapel_spike_jnt}
    constraintSetup [
     {targetBone {boneName=spine_05}}
     {targetBone {boneName=clavicle_r}}
     {targetBone {boneName=spine_04}}
     {targetBone {boneName=upperarm_twist_01_r}}
     {targetBone {boneName=deform_upperarm_fr_r}}
     {targetBone {boneName=deform_upperarm_up_r}}
     {targetBone {boneName=deform_pec_r}}
    ]
    constraintWeights [1,1,1,1,1,1,1]
   }
   n21 = "Constraint - back_collar_spike_jnt (PPPPPPP)"!(
    ComponentPose=@n20
    Alpha=1
    ConstraintWeights_0=0.0618
    ConstraintWeights_1:float=0.1553
    ConstraintWeights_2:float=0.0806
    ConstraintWeights_3:float=0.4918
    ConstraintWeights_4:float=0.1101
    ConstraintWeights_5:float=0.0885
    ConstraintWeights_6:float=0.0119
   ) -> (Pose) {
    alphaBoolBlend {blendOption=Linear}
    boneToModify {boneName=back_collar_spike_jnt}
    constraintSetup [
     {targetBone {boneName=head}}
     {targetBone {boneName=neck_02}}
     {targetBone {boneName=neck_01}}
     {targetBone {boneName=spine_05}}
     {targetBone {boneName=clavicle_l}}
     {targetBone {boneName=clavicle_r}}
     {targetBone {boneName=upperarm_twist_01_r}}
    ]
    constraintWeights [1,1,1,1,1,1,1]
   }
   n22 = "Constraint - L_collar_spike_jnt (PPPPPPP)"!(
    ComponentPose=@n21
    Alpha=1
    ConstraintWeights_0=0.0285
    ConstraintWeights_1:float=0.0827
    ConstraintWeights_2:float=0.0421
    ConstraintWeights_3:float=0.3222
    ConstraintWeights_4:float=0.4497
    ConstraintWeights_5:float=0.0491
    ConstraintWeights_6:float=0.0256
   ) -> (Pose) {
    alphaBoolBlend {blendOption=Linear}
    boneToModify {boneName=L_collar_spike_jnt}
    constraintSetup [
     {targetBone {boneName=head}}
     {targetBone {boneName=neck_02}}
     {targetBone {boneName=neck_01}}
     {targetBone {boneName=spine_05}}
     {targetBone {boneName=clavicle_l}}
     {targetBone {boneName=upperarm_twist_01_l}}
     {targetBone {boneName=deform_upperarm_up_l}}
    ]
    constraintWeights [1,1,1,1,1,1,1]
   }
   n23 = "Constraint - R_collar_spike_jnt (PPPPPPP)"!(
    ComponentPose=@n22
    Alpha=1
    ConstraintWeights_0=0.0268
    ConstraintWeights_1:float=0.0776
    ConstraintWeights_2:float=0.0395
    ConstraintWeights_3:float=0.3453
    ConstraintWeights_4:float=0.4407
    ConstraintWeights_5:float=0.0461
    ConstraintWeights_6:float=0.024
   ) -> (Pose) {
    alphaBoolBlend {blendOption=Linear}
    boneToModify {boneName=R_collar_spike_jnt}
    constraintSetup [
     {targetBone {boneName=head}}
     {targetBone {boneName=neck_02}}
     {targetBone {boneName=neck_01}}
     {targetBone {boneName=spine_05}}
     {targetBone {boneName=clavicle_r}}
     {targetBone {boneName=upperarm_twist_01_r}}
     {targetBone {boneName=deform_upperarm_up_r}}
    ]
    constraintWeights [1,1,1,1,1,1,1]
   }
   Output Pose(Result=Component To Local(ComponentPose=@n23)) {layerGroup=None}
   n24 = Control Rig(Source=Copy Pose From Mesh() {bUseAttachedParent=true}) {
    alphaBoolBlend {blendOption=Linear}
    controlRigAssetReference {blueprintRigClass=&/F_MED_DeformRig_PinkSpike_ControlRig.F_MED_DeformRig_PinkSpike_ControlRig_C}
   }
   Save cached pose 'PinkSpikePreSim'(Pose=@n24) {cacheName=PinkSpikePreSim}
  ]}
 EventGraph {nodes [
   Event BlueprintUpdateAnimation()
   > Set Body Mesh(Body Mesh:Skeletal Mesh Component Object Reference=GetOwnerSkeletalMeshComponent()) -> (Output_Get:Skeletal Mesh Component Object Reference)
   > Is Valid(InputObject=Get Body Mesh() -> (Body Mesh:Skeletal Mesh Component Object Reference))
   > Set Jacket Front Side Correction(Jacket Front Side Correction:double=GetAlphaFromReferenceTransformDelta(
     SkeletalMeshComponent=Get Body Mesh()
     BoneName=thigh_l
     BaseName=pelvis
     DeltaComponentType=QuaternionTwist
     InputRange=(X=40,Y=106)
     TwistAxis="0,0,1"
    )) -> (Output_Get:double)
   n1 = Get bIsSkyDiving()
   n2 = Get bIsSkydiveFloating()
   > Set Skydiving(Skydiving:double=SelectFloat(A=1, B=0, bPickA=OR Boolean(A=@n1, B=@n2))) -> (Output_Get:double)
   n3 = GetAnimInstance(self=GetOwnerSkeletalMeshComponent())
   > Is Valid(InputObject=@n3)
   n4 = Montage_GetPosition!(
    self:Anim Instance Object Reference=@n3
    Montage:Anim Montage Object Reference=/BRCosmetics/Animation/Game/MainPlayer/Emotes/PinkSpike/Emote_PinkSpike_CMF_M
   ) -> (ReturnValue:float)
   > Set Is Custom Emote Playing(Is Custom Emote Playing:double=SelectFloat(A=1, B=0, bPickA="float > float"(A=@n4, B=1.5))) -> (Output_Get:double)
   > Set Is In Front End(Is In Front End:double="To Float (Boolean)"(InBool=Get bIsInFrontEnd())) -> (Output_Get:double)
   float * float(A:Wildcard, B:Wildcard) -> (ReturnValue:Wildcard) {is=title K2Node_PromotableOperator}
  ]}
]
variables [
 Body Mesh {type=SkeletalMeshComponent}
 Is Custom Emote Playing {type=number}
 Is In Front End {type=number}
 Jacket Front Side Correction {type=number}
 Skydiving {type=number}
 Weapon Change {type=number}
]
