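# Physics asset for the F_MED_PinkSpike face accessory, defining the simulation rig for the character's pigtails, side fringes and bangs.
# Capsule chains hang from master bones with tight swing and twist limits, with strands branching off each main pigtail's third joint.
# Head, neck, upper spine and upper-arm capsules are linked by free constraints.
# The sides are tuned differently: left main pigtail joints swing wider than the right.
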PhysicsAsset /BRCosmetics/Characters/Player/Female/Medium/Bodies/F_MED_PinkSpike/Meshes/Parts/F_MED_PinkSpike_FaceAcc_Physics
bodies {
 L_pigtail_master [{shapeType=Capsule, length=0.505, radius=0.505, shapeName=L_pigtail_master_capsule}]
 L_sidefringe_master_jnt [{shapeType=Capsule, length=0.505, radius=0.505, shapeName=L_sidefringe_master_jnt_capsule}]
 R_pigtail_master [{shapeType=Capsule, length=0.505, radius=0.505, shapeName=R_pigtail_master_capsule}]
 R_sidefringe_master_jnt [{shapeType=Capsule, length=0.505, radius=0.505, shapeName=R_sidefringe_master_jnt_capsule}]
 bangs_master_jnt [{shapeType=Capsule, length=0.505, radius=0.505, shapeName=bangs_master_jnt_capsule}]
 dyn_L_bangs_jnt [{
   shapeType=Capsule
   length=1.7094
   radius=2.4472
   shapeName=dyn_L_bangs_jnt_capsule
   center=<2.0225 -0.2944 1.6411>
   rotation=<rot 0.0001 0.0002 -80>
  }]
 dyn_L_main_pigtail_1_jnt [{
   shapeType=Capsule
   length=18.1737
   radius=5.4918
   shapeName=dyn_L_main_pigtail_1_jnt_capsule
   center=<13.0036 -3.0582 2.8378>
   rotation=<rot 90 0 0>
  }]
 dyn_L_main_pigtail_2_jnt [{
   shapeType=Capsule
   length=17.4989
   radius=6.322
   shapeName=dyn_L_main_pigtail_2_jnt_capsule
   center=<13.8387 -2.8144 0.7911>
   rotation=<rot 90 0 0>
  }]
 dyn_L_main_pigtail_3_jnt [{
   shapeType=Capsule
   length=19.6087
   radius=6.3151
   shapeName=dyn_L_main_pigtail_3_jnt_capsule
   center=<15.768 -3.4655 0.8036>
   rotation=<rot 90 0 0>
  }]
 dyn_L_main_pigtail_4_jnt [{
   shapeType=Capsule
   length=16.8064
   radius=6.1115
   shapeName=dyn_L_main_pigtail_4_jnt_capsule
   center=<13.3846 -1.4583 0>
   rotation=<rot 90 0 0>
  }]
 dyn_L_main_pigtail_5_jnt [{
   shapeType=Capsule
   length=12.0891
   radius=5.3216
   shapeName=dyn_L_main_pigtail_5_jnt_capsule
   center=<7.203 1.9209 -2.1386>
   rotation=<rot 90 0 0>
  }]
 dyn_L_sidefringe_1_jnt [{shapeType=Capsule, length=2, radius=2, shapeName=dyn_L_sidefringe_1_jnt_capsule, center=<2.3982 0 1.3364>, rotation=<rot -80 -179.9998 -180>}]
 dyn_L_sidefringe_2_jnt [{shapeType=Capsule, length=2, radius=2, shapeName=dyn_L_sidefringe_2_jnt_capsule, center=<2.3982 0 1.3364>, rotation=<rot -80 -179.9998 180>}]
 dyn_L_strand_pigtail_1_jnt [{
   shapeType=Capsule
   length=4.9496
   radius=2.1977
   shapeName=dyn_L_strand_pigtail_1_jnt_capsule
   center=<4.3311 0 0>
   rotation=<rot 90 179.9209 179.9207>
  }]
 dyn_L_strand_pigtail_2_jnt [{
   shapeType=Capsule
   length=3.4243
   radius=2.7659
   shapeName=dyn_L_strand_pigtail_2_jnt_capsule
   center=<4.3311 0 0>
   rotation=<rot 90 -179.7889 -179.7893>
  }]
 dyn_L_strand_pigtail_3_jnt [{
   shapeType=Capsule
   length=6.3798
   radius=3.1463
   shapeName=dyn_L_strand_pigtail_3_jnt_capsule
   center=<4.3311 0 0>
   rotation=<rot 90 -90 -90.0006>
  }]
 dyn_R_bangs_jnt [{shapeType=Capsule, length=1.7094, radius=2.4472, shapeName=dyn_L_bangs_jnt_capsule, center=<-1.8579 0 -1.0454>, rotation=<rot 0 0 90>}]
 dyn_R_main_pigtail_1_jnt [{
   shapeType=Capsule
   length=17.25
   radius=6
   shapeName=dyn_R_main_pigtail_1_jnt_capsule
   center=<-14.3472 4.0213 -2.7244>
   rotation=<rot 90 0 0>
  }]
 dyn_R_main_pigtail_2_jnt [{shapeType=Capsule, length=17.25, radius=6, shapeName=dyn_R_main_pigtail_2_jnt_capsule, center=<-15.3194 2.9586 0>, rotation=<rot 90 0 0>}]
 dyn_R_main_pigtail_3_jnt [{shapeType=Capsule, length=17.25, radius=6, shapeName=dyn_R_main_pigtail_3_jnt_capsule, center=<-15.3194 3.7456 0>, rotation=<rot 90 0 0>}]
 dyn_R_main_pigtail_4_jnt [{shapeType=Capsule, length=16, radius=6, shapeName=dyn_R_main_pigtail_4_jnt_capsule, center=<-12.9672 1.935 0>, rotation=<rot 90 0 0>}]
 dyn_R_main_pigtail_5_jnt [{
   shapeType=Capsule
   length=10
   radius=6.5
   shapeName=dyn_R_main_pigtail_5_jnt_capsule
   center=<-9.6831 2.4219 1.7305>
   rotation=<rot 90 172.875 -167.125>
  }]
 dyn_R_sidefringe_1_jnt [{shapeType=Capsule, length=2, radius=2, shapeName=dyn_R_sidefringe_1_jnt_capsule, center=<-2.4913 0 -1.2528>, rotation=<rot 90 180 179.9997>}]
 dyn_R_sidefringe_2_jnt [{
   shapeType=Capsule
   length=2
   radius=2
   shapeName=dyn_R_sidefringe_2_jnt_capsule
   center=<-2.4913 0 -1.2528>
   rotation=<rot 90 179.9963 179.9961>
  }]
 dyn_R_strand_pigtail_1_jnt [{
   shapeType=Capsule
   length=3
   radius=3
   shapeName=dyn_R_strand_pigtail_1_jnt_capsule
   center=<-4.0176 0 0>
   rotation=<rot -90 -179.9999 -179.9997>
  }]
 dyn_R_strand_pigtail_2_jnt [{
   shapeType=Capsule
   length=3
   radius=3
   shapeName=dyn_R_strand_pigtail_2_jnt_capsule
   center=<-4.0176 0 0>
   rotation=<rot -90 -179.8812 179.8816>
  }]
 dyn_R_strand_pigtail_3_jnt [{
   shapeType=Capsule
   length=3
   radius=3
   shapeName=dyn_R_strand_pigtail_3_jnt_capsule
   center=<-4.0176 0 0>
   rotation=<rot -90 -166.7373 166.7376>
  }]
 dyn_center_bangs_jnt [{shapeType=Capsule, length=6.2661, radius=2.2023, shapeName=dyn_center_bangs_jnt_capsule, center=<2.1853 1.6775 0>}]
 dyn_top_bangs_jnt [{
   shapeType=Capsule
   length=3.527
   radius=2.3268
   shapeName=dyn_top_bangs_jnt_capsule
   center=<0.8252 1.5658 0>
   rotation=<rot 0 0.001 -0.0001>
  }]
 head [{shapeType=Capsule, length=4, radius=7, shapeName=head_capsule, center=<3.3171 1.8859 -0.2871>, rotation=<rot 1.0102 -90.0001 89.7539>}]
 master_hair_jnt [{shapeType=Capsule, length=0.505, radius=0.505, shapeName=master_hair_jnt_capsule}]
 neck_01 [{shapeType=Capsule, length=95.5, radius=6.5, shapeName=neck_01_capsule, center=<2.5307 4.6714 0>, rotation=<rot 0 -90 0>}]
 neck_02 [{shapeType=Capsule, length=5.25, radius=5.5, shapeName=neck_02_capsule, rotation=<rot 0 -90.0001 90>}]
 spine_04 [{shapeType=Capsule, length=8.5, radius=11, shapeName=spine_04_capsule}]
 spine_05 [{shapeType=Capsule, length=13.505, radius=11.755, shapeName=spine_05_capsule, center=<1.9121 -0.7232 0.6885>, rotation=<rot 0 -90 0>}]
 upperarm_twist_01_l [{shapeType=Capsule, length=23, radius=6, shapeName=upperarm_twist_01_l_capsule, center=<2.2987 0 -3.129>, rotation=<rot -0.0003 -89.9997 -60>}]
 upperarm_twist_01_r [{
   shapeType=Capsule
   length=41.75
   radius=6
   shapeName=upperarm_twist_01_r_capsule
   center=<-7.4507 0 1.3138>
   rotation=<rot -0.0002 -89.9998 -80>
  }]
}
constraints [
 {bone1Name=dyn_L_main_pigtail_1_jnt, bone2Name=L_pigtail_master, swing1LimitDegrees=20, swing2LimitDegrees=10, twistLimitDegrees=1}
 {bone1Name=dyn_L_main_pigtail_2_jnt, bone2Name=dyn_L_main_pigtail_1_jnt, swing1LimitDegrees=20, swing2LimitDegrees=10, twistLimitDegrees=1}
 {bone1Name=dyn_L_main_pigtail_3_jnt, bone2Name=dyn_L_main_pigtail_2_jnt, swing1LimitDegrees=20, swing2LimitDegrees=10, twistLimitDegrees=1}
 {bone1Name=dyn_L_main_pigtail_4_jnt, bone2Name=dyn_L_main_pigtail_3_jnt, swing1LimitDegrees=20, swing2LimitDegrees=10, twistLimitDegrees=1}
 {bone1Name=dyn_L_main_pigtail_5_jnt, bone2Name=dyn_L_main_pigtail_4_jnt, swing1LimitDegrees=20, swing2LimitDegrees=10, twistLimitDegrees=1}
 {bone1Name=dyn_L_strand_pigtail_1_jnt, bone2Name=dyn_L_main_pigtail_3_jnt, swing1LimitDegrees=5, swing2LimitDegrees=5, twistLimitDegrees=1}
 {bone1Name=dyn_L_strand_pigtail_2_jnt, bone2Name=dyn_L_strand_pigtail_1_jnt, swing1LimitDegrees=5, swing2LimitDegrees=5, twistLimitDegrees=1}
 {bone1Name=dyn_L_strand_pigtail_3_jnt, bone2Name=dyn_L_strand_pigtail_2_jnt, swing1LimitDegrees=5, swing2LimitDegrees=5, twistLimitDegrees=1}
 {
  bone1Name=L_sidefringe_master_jnt
  bone2Name=master_hair_jnt
  swing1LimitDegrees=45
  swing1Motion=Free
  swing2LimitDegrees=45
  swing2Motion=Free
  twistLimitDegrees=45
  twistMotion=Free
 }
 {bone1Name=dyn_L_sidefringe_1_jnt, bone2Name=L_sidefringe_master_jnt, swing1LimitDegrees=25, swing2LimitDegrees=25, twistLimitDegrees=1}
 {bone1Name=dyn_L_sidefringe_2_jnt, bone2Name=dyn_L_sidefringe_1_jnt, swing1LimitDegrees=25, swing2LimitDegrees=25, twistLimitDegrees=1}
 {bone1Name=dyn_R_main_pigtail_1_jnt, bone2Name=R_pigtail_master, swing1LimitDegrees=5, swing2LimitDegrees=5, twistLimitDegrees=1}
 {bone1Name=dyn_R_main_pigtail_2_jnt, bone2Name=dyn_R_main_pigtail_1_jnt, swing1LimitDegrees=5, swing2LimitDegrees=5, twistLimitDegrees=1}
 {bone1Name=dyn_R_main_pigtail_3_jnt, bone2Name=dyn_R_main_pigtail_2_jnt, swing1LimitDegrees=5, swing2LimitDegrees=5, twistLimitDegrees=1}
 {bone1Name=dyn_R_main_pigtail_4_jnt, bone2Name=dyn_R_main_pigtail_3_jnt, swing1LimitDegrees=10, swing2LimitDegrees=10, twistLimitDegrees=1}
 {bone1Name=dyn_R_main_pigtail_5_jnt, bone2Name=dyn_R_main_pigtail_4_jnt, swing1LimitDegrees=10, swing2LimitDegrees=10, twistLimitDegrees=1}
 {bone1Name=dyn_R_strand_pigtail_1_jnt, bone2Name=dyn_R_main_pigtail_3_jnt, swing1LimitDegrees=5, swing2LimitDegrees=5, twistLimitDegrees=1}
 {bone1Name=dyn_R_strand_pigtail_2_jnt, bone2Name=dyn_R_strand_pigtail_1_jnt, swing1LimitDegrees=5, swing2LimitDegrees=5, twistLimitDegrees=1}
 {bone1Name=dyn_R_strand_pigtail_3_jnt, bone2Name=dyn_R_strand_pigtail_2_jnt, swing1LimitDegrees=5, swing2LimitDegrees=5, twistLimitDegrees=1}
 {
  bone1Name=R_sidefringe_master_jnt
  bone2Name=master_hair_jnt
  swing1LimitDegrees=45
  swing1Motion=Free
  swing2LimitDegrees=45
  swing2Motion=Free
  twistLimitDegrees=45
  twistMotion=Free
 }
 {bone1Name=dyn_R_sidefringe_1_jnt, bone2Name=R_sidefringe_master_jnt, swing1LimitDegrees=25, swing2LimitDegrees=25, twistLimitDegrees=1}
 {bone1Name=dyn_R_sidefringe_2_jnt, bone2Name=dyn_R_sidefringe_1_jnt, swing1LimitDegrees=25, swing2LimitDegrees=25, twistLimitDegrees=1}
 {
  bone1Name=bangs_master_jnt
  bone2Name=master_hair_jnt
  swing1LimitDegrees=45
  swing1Motion=Free
  swing2LimitDegrees=45
  swing2Motion=Free
  twistLimitDegrees=45
  twistMotion=Free
 }
 {bone1Name=dyn_top_bangs_jnt, bone2Name=bangs_master_jnt, swing1LimitDegrees=2, swing2LimitDegrees=5, twistLimitDegrees=1}
 {bone1Name=dyn_L_bangs_jnt, bone2Name=dyn_top_bangs_jnt, swing1LimitDegrees=2, swing2LimitDegrees=5, twistLimitDegrees=1}
 {bone1Name=dyn_R_bangs_jnt, bone2Name=dyn_top_bangs_jnt, swing1LimitDegrees=2, swing2LimitDegrees=5, twistLimitDegrees=1}
 {bone1Name=dyn_center_bangs_jnt, bone2Name=dyn_top_bangs_jnt, swing1LimitDegrees=2, swing2LimitDegrees=5, twistLimitDegrees=1}
 {
  bone1Name=master_hair_jnt
  bone2Name=head
  swing1LimitDegrees=45
  swing1Motion=Free
  swing2LimitDegrees=45
  swing2Motion=Free
  twistLimitDegrees=45
  twistMotion=Free
 }
 {
  bone1Name=neck_02
  bone2Name=neck_01
  swing1LimitDegrees=45
  swing1Motion=Free
  swing2LimitDegrees=45
  swing2Motion=Free
  twistLimitDegrees=45
  twistMotion=Free
 }
 {
  bone1Name=head
  bone2Name=neck_02
  swing1LimitDegrees=45
  swing1Motion=Free
  swing2LimitDegrees=45
  swing2Motion=Free
  twistLimitDegrees=45
  twistMotion=Free
 }
 {
  bone1Name=spine_05
  bone2Name=spine_04
  swing1LimitDegrees=45
  swing1Motion=Free
  swing2LimitDegrees=45
  swing2Motion=Free
  twistLimitDegrees=45
  twistMotion=Free
 }
 {
  bone1Name=upperarm_twist_01_r
  bone2Name=spine_05
  swing1LimitDegrees=45
  swing1Motion=Free
  swing2LimitDegrees=45
  swing2Motion=Free
  twistLimitDegrees=45
  twistMotion=Free
 }
 {
  bone1Name=upperarm_twist_01_l
  bone2Name=spine_05
  swing1LimitDegrees=45
  swing1Motion=Free
  swing2LimitDegrees=45
  swing2Motion=Free
  twistLimitDegrees=45
  twistMotion=Free
 }
 {
  bone1Name=neck_01
  bone2Name=spine_05
  swing1LimitDegrees=45
  swing1Motion=Free
  swing2LimitDegrees=45
  swing2Motion=Free
  twistLimitDegrees=45
  twistMotion=Free
 }
]
