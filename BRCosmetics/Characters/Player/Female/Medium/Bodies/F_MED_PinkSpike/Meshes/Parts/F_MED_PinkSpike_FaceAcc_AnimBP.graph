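# Post-copy hair AnimBP for the F_MED_PinkSpike face accessory, adding pigtail physics and pose corrections.
# The parent mesh pose is cached and feeds two head-space RigidBody branches, selected by backbling state to swap physics assets.
# Additive bone transforms then offset pigtails for head rotation, tactical sprint and skydiving without parachute.
# Crouched aiming nudges the right pigtail before simulation and lowers physics alpha.
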AnimBlueprint /BRCosmetics/Characters/Player/Female/Medium/Bodies/F_MED_PinkSpike/Meshes/Parts/F_MED_PinkSpike_FaceAcc_AnimBP
target_skeleton=&/F_MED_PinkSpike_FaceAcc_Skeleton
graphs [
 AnimGraph {nodes [
   n1 = Get TacSprint() -> (TacSprint:double)
   n2 = "Transform (Modify) Bone - Bone: R_pigtail_master"(
    ComponentPose=Local To Component(LocalPose=Use cached pose 'PinkSpikeHair'())
    Alpha=Get CrouchingADS() -> (CrouchingADS:double)
    Translation="0,0,2"
    Rotation="-10,0,0"
    Scale="1,1,1"
   ) {
    rotationMode=BMM_Additive
    rotationSpace=BCS_BoneSpace
    translationMode=BMM_Additive
    translationSpace=BCS_BoneSpace
    alphaBoolBlend {blendOption=Linear}
    boneToModify {boneName=R_pigtail_master}
    rotation {pitch=-10}
    translation {z=2}
   }
   n3 = Get CrouchingADS()
   n4 = RigidBody(ComponentPose=@n2, Alpha=SelectFloat(A=0.7, B=1, bPickA="Equal (Float)"(A=@n3, B=1))) {
    bClampLinearTranslationLimitToRefPose=true
    bOverrideWorldGravity=true
    overridePhysicsAsset=&/F_MED_PinkSpike_FaceAcc_Physics
    simulationSpace=BaseBoneSpace
    alphaBoolBlend {blendOption=Linear}
    baseBoneRef {boneName=head}
    componentAppliedLinearAccClamp=<3000 3000 3000>
    componentLinearAccScale=<0.45 0.2 0.2>
    componentLinearVelScale=<0.45 0.2 0.2>
    overrideWorldGravity {z=-980}
    simSpaceSettings {maxAngularAcceleration=50, maxAngularVelocity=4, maxLinearAcceleration=300, maxLinearVelocity=200, velocityScaleZ=0.3, worldAlpha=1}
   }
   n5 = Component To Local(ComponentPose=@n4)
   n6 = "Transform (Modify) Bone - Bone: R_pigtail_master"(
    ComponentPose=Local To Component(LocalPose=Use cached pose 'PinkSpikeHair'())
    Alpha=Get CrouchingADS()
    Translation="0,0,2"
    Rotation="-10,0,0"
    Scale="1,1,1"
   ) {
    rotationMode=BMM_Additive
    rotationSpace=BCS_BoneSpace
    translationMode=BMM_Additive
    translationSpace=BCS_BoneSpace
    alphaBoolBlend {blendOption=Linear}
    boneToModify {boneName=R_pigtail_master}
    rotation {pitch=-10}
    translation {z=2}
   }
   n7 = Get CrouchingADS()
   n8 = RigidBody(ComponentPose=@n6, Alpha=SelectFloat(A=0.7, B=1, bPickA="Equal (Float)"(A=@n7, B=1))) {
    bClampLinearTranslationLimitToRefPose=true
    bOverrideWorldGravity=true
    overridePhysicsAsset=&/F_MED_PinkSpike_Backbling_FaceAcc_Physics
    simulationSpace=BaseBoneSpace
    alphaBoolBlend {blendOption=Linear}
    baseBoneRef {boneName=head}
    componentAppliedLinearAccClamp=<3000 3000 3000>
    componentLinearAccScale=<0.45 0.2 0.2>
    componentLinearVelScale=<0.45 0.2 0.2>
    overrideWorldGravity {z=-980}
    simSpaceSettings {maxAngularAcceleration=50, maxAngularVelocity=4, maxLinearAcceleration=300, maxLinearVelocity=200, velocityScaleZ=0.3, worldAlpha=1}
   }
   n9 = Component To Local(ComponentPose=@n8)
   n10 = Get Backbling() -> (Backbling:bool)
   n11 = SelectFloat(A=0, B=Get Skydiving() -> (Skydiving:double), bPickA=Get Parachute() -> (Parachute:bool))
   n12 = "Transform (Modify) Bone - Bone: L_pigtail_master"(
    ComponentPose=Local To Component(LocalPose=Blend Poses by bool(BlendPose_0=@n9, BlendPose_1=@n5, bActiveValue=@n10) {blendType=Linear})
    Alpha=@n11
    Translation="0,0,0"
    Rotation="20,-54.000,0"
    Scale="1,1,1"
   ) {
    rotationMode=BMM_Additive
    rotationSpace=BCS_BoneSpace
    alphaBoolBlend {blendOption=Linear}
    boneToModify {boneName=L_pigtail_master}
    rotation {pitch=20, yaw=-54}
   }
   n13 = "Transform (Modify) Bone - Bone: R_pigtail_master"(ComponentPose=@n12, Alpha=@n11, Translation="0,0,0", Rotation="20,-50,0", Scale="1,1,1") {
    rotationMode=BMM_Additive
    rotationSpace=BCS_BoneSpace
    alphaBoolBlend {blendOption=Linear}
    boneToModify {boneName=R_pigtail_master}
    rotation {pitch=20, yaw=-50}
   }
   n14 = "Transform (Modify) Bone - Bone: R_pigtail_master"(ComponentPose=@n13, Alpha=Get Head Back Rot() -> (Head Back Rot:double), Translation="0,0,0", Rotation="0,-10,0", Scale="1,1,1") {
    rotationMode=BMM_Additive
    rotationSpace=BCS_BoneSpace
    alphaBoolBlend {blendOption=Linear}
    boneToModify {boneName=R_pigtail_master}
    rotation {yaw=-10}
   }
   n15 = "Transform (Modify) Bone - Bone: R_pigtail_master"(ComponentPose=@n14, Alpha=Get Head R Rot() -> (Head R Rot:double), Translation="0,0,0", Rotation="0,-30.000,0", Scale="1,1,1") {
    rotationMode=BMM_Additive
    rotationSpace=BCS_BoneSpace
    alphaBoolBlend {blendOption=Linear}
    boneToModify {boneName=R_pigtail_master}
    rotation {yaw=-30}
   }
   n16 = "Transform (Modify) Bone - Bone: R_pigtail_master"(ComponentPose=@n15, Alpha=@n1, Translation="0,0,0", Rotation="10.000,-10.000,-10.000", Scale="1,1,1") {
    rotationMode=BMM_Additive
    rotationSpace=BCS_BoneSpace
    alphaBoolBlend {blendOption=Linear}
    alphaScaleBiasClamp {bInterpResult=true, interpSpeedDecreasing=2, interpSpeedIncreasing=2}
    boneToModify {boneName=R_pigtail_master}
    rotation=<rot 10 -9.9999 -10>
   }
   n17 = "Transform (Modify) Bone - Bone: L_pigtail_master"(ComponentPose=@n16, Alpha=@n1, Translation="0,0,0", Rotation="0,-20,-10", Scale="1,1,1") {
    rotationMode=BMM_Additive
    rotationSpace=BCS_BoneSpace
    alphaBoolBlend {blendOption=Linear}
    alphaScaleBiasClamp {bInterpResult=true, interpSpeedDecreasing=2, interpSpeedIncreasing=2}
    boneToModify {boneName=L_pigtail_master}
    rotation {roll=-10, yaw=-20}
   }
   Output Pose(Result=Component To Local(ComponentPose=@n17)) {layerGroup=None}
   Save cached pose 'PinkSpikeHair'(Pose=Copy Pose From Mesh() {bUseAttachedParent=true}) {cacheName=PinkSpikeHair}
  ]}
 EventGraph {nodes [
   Event BlueprintUpdateAnimation()
   > n1 = Sequence()
   n2 = GetOwnerSkeletalMeshComponent()
   n3 = BreakRotator(InRot=BreakTransform(InTransform=GetDeltaTransformFromRefPose(self=@n2, BoneName=head, BaseName=spine_05)).Rotation)
   > Set Head Back Rot(Head Back Rot:double=MapRangeClamped(Value=@n3.Yaw, InRangeA=-10, InRangeB=-55)) -> (Output_Get:double)
   > Set Head R Rot(Head R Rot:double=MapRangeClamped(Value=@n3.Pitch, InRangeA=0, InRangeB=45)) -> (Output_Get:double)
   n4 = Get bIsSkyDiving()
   n5 = Get bIsSkydiveFloating()
   > Set Skydiving(Skydiving:double=SelectFloat(A=1, B=0, bPickA=OR Boolean(A=@n4, B=@n5))) -> (Output_Get:double)
   > n6 = Set Parachute(Parachute:bool=Get bIsParachuteOpened()) -> (Output_Get:bool)
   n7 = CheckForBackbling()
   Set Backbling(execute=[@n10.then, @n6.then], Backbling:bool=OR Boolean(A=@n7, B=@n7.bIsBackblingLong)) -> (Output_Get:bool)
   Set IsCrouching(execute=@n1.then_1, IsCrouching:double=SelectFloat(A=1, B=0, bPickA=Get bIsCrouching())) -> (Output_Get:double)
   n8 = Get bIsCrouching()
   n9 = Get bIsTargeting()
   > Set CrouchingADS(CrouchingADS:double=SelectFloat(A=1, B=0, bPickA=AND Boolean(A=@n9, B=@n8))) -> (Output_Get:double)
   > Set TacSprint(TacSprint:double=SelectFloat(A=1, B=0, bPickA=Get bIsTacticalSprinting())) -> (Output_Get:double)
   n10 = Event OnCharacterCustomizationLoaded()
  ]}
]
variables [
 Backbling {type=boolean}
 CrouchingADS {type=number}
 Head Back Rot {type=number}
 Head R Rot {type=number}
 IsCrouching {type=number}
 Parachute {type=boolean}
 Skydiving {type=number}
 TacSprint {type=number}
]
